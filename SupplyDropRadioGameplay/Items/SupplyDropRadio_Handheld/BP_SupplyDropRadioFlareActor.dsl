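; Flare actor for the handheld supply drop radio: it is popped, thrown as a projectile, and spawns a supply drop at its location.
; The radio weapon calls pop and throw; replicated state drives flare FX, collision, water interaction and a rotation lerp.
; On the server, bounces are capped, and projectile stop or a maximum simulation timeout starts the spawn timer.
; The drop class is async-loaded using the conversation index cached from the game state component.
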(event EventBeginPlay
  ;[0.05611 0.5781 0.1377] BeginPlay: Setting this actor to ignore collision with the owner pawn.
  (|CallParentFunction)
  (Collision|IgnoreActorwhenMoving (Variables|Default|GetMesh) (Actor|GetOwner) true))

(event OnProjectileBounce(FortProjectileMovement)
  (params (ImpactResult :type Hit%20Result%20Structure%20%28by%20ref%29) (ImpactVelocity :type Vector%20%28by%20ref%29 :default "0, 0, 0"))
  (bind (_bblockinghit _binitialoverlap _time _distance _location _impactpoint _normal _impactnormal) (Collision|BreakHitResult ImpactResult))
  ;[0.01951 0.09189 0.09375] OnProjectileBounce: On the server, track the bounce count and zero out the bounciness after the desired number of bounces.
  ; On the client, play the bounce sound.
  (Networking|SwitchHasAuthority
    (:Authority
      (bind _result (Math|Integer|IncrementInt (Variables|Transient|GetBounces))
        (:%20%20
          (if (> _result (ScalableFloat|GetValueasInteger (Variables|Default|GetMaxAllowedBounces)))
            (Variables|/Script/Engine.ProjectileMovementComponent|SetBounciness 0.0 (Variables|Default|GetFortProjectileMovement)) ; Stop movement on the next bounce.
          ))))
    (:Remote
      (if (not (Water|IsinWater (Variables|Default|GetFortWaterInteraction)))
        (Audio|OnCollision (Variables|Default|GetFortCollisionAudio) _impactpoint (* _impactnormal (Variables|/Script/Engine.MovementComponent|GetVelocity (Variables|Default|GetFortProjectileMovement))))))))

(event OnProjectileStop(FortProjectileMovement)
  (params (ImpactResult :type Hit%20Result%20Structure%20%28by%20ref%29))
  ;[0.1719 0.03766 0.1407] OnProjectileStop: Start a timer after the projectile has stopped moving, after which the supply drop actor is spawned.
  (Networking|SwitchHasAuthority
    (:Authority
      (CallFunction|StartSpawnSupplyDropTimer)
      (Utilities|Time|ClearandInvalidateTimerbyHandle (Variables|Transient|GetProjectileSimulationTimerHandle)))))

(event Custom|OnRep_ThrowRotation
  ;[0.07891 0.2654 0.2708] OnRep_ThrowRotation: Once the throw rotation is set, lerp the actor rotation to that desired rotation value.
  ; On both client and server.
  (Variables|Transient|SetInitialRotation (Transformation|GetActorRotation))
  (Components|Timeline|SetPlayRate (Variables|BP_SupplyDropRadioFlareActor|GetTimeline_ThrowRotation) (/ 1.0 (Variables|Default|GetThrowRotationLerpDuration)))
  (bind (_direction _alpha) (Timeline|Timeline_ThrowRotation Play)
    (:Update
      (Transformation|SetActorRotation (Math|Rotator|Lerp(Rotator) (Variables|Transient|GetInitialRotation) (Variables|Transient|GetThrowRotation) _alpha true)))))

(event OnEnterWaterEvent(FortWaterInteraction)
  (params (WaterBody :type Fort%20Water%20Body%20Actor%20Object%20Reference) (WaterInteractionComponent :type Fort%20Water%20Interaction%20Component%20Object%20Reference) (bIsFirstBody :type Boolean))
  ;[0.2635 0.03734 0.2708] OnEnterWaterEvent: Play the splash FX (Visual and sound) on the water.
  (Networking|CanExecuteCosmeticEvents
    (:True
      (bind _outwatersurfaceinfo (WaterBody|GetWaterSurfaceInfoStruct WaterBody (Transformation|GetActorLocation) true false false))
      (bind _ns_watersplash (Variables|Default|GetNS_WaterSplash)
        (:then
          (bind _returnvalue (Effects|Components|ParticleSystem|SpawnFXAtLocation _ns_watersplash (Utilities|Struct|BreakWaterSurfaceInfo _outwatersurfaceinfo) (Math|Rotator|MakeRotfromX (.WaterPlaneNormal (Utilities|Struct|BreakWaterSurfaceInfo _outwatersurfaceinfo)))))
          (Utilities|IsValid _returnvalue
            (:Is%20Valid
              (Effects|Components|ParticleSystem|SetFloatParameter _returnvalue "WaterDepth" (.WaterDepth (Utilities|Struct|BreakWaterSurfaceInfo _outwatersurfaceinfo)))
              (Effects|Components|ParticleSystem|SetFloatParameter _returnvalue "WaterBodyIndex" (Math|Conversions|ToFloat(Integer) (.WaterBodyIdx (Utilities|Struct|BreakWaterSurfaceInfo _outwatersurfaceinfo))))
              (Audio|PlaySoundatLocation (Variables|Default|GetWaterSplashSoundCue) (Transformation|GetActorLocation)))))
        (:else)))))

(event Custom|ExecutePopCosmetic
  (bind (_returnvalue_x _returnvalue_y) (Math|Vector|BreakVector (* (Math|Vector|RotateVectorAroundAxis (Transformation|GetActorForwardVector (Actor|GetOwner)) 3e+01 "0, 0,1.000000") 2e+02)))
  ;[0.03869 0.1861 0.1979] ExecutePopCosmetic: Pop the flare!
  (Networking|CanExecuteCosmeticEvents
    (:True
      (Utilities|FlowControl|DoOnce false
        (:Completed
          (Components|StaticMesh|SetStaticMesh (Variables|Default|GetMesh) "/SupplyDropRadioGameplay/Items/SupplyDropRadio_Handheld/Meshes/Supply_Radio_Flare_Stick.Supply_Radio_Flare_Stick")
          (bind _returnvalue (Niagara|SpawnSystemAttached (Variables|Default|GetFlare%20Niagara%20System%20Template) (Variables|Default|GetMesh) "FXSocket" "0, 0, 0" (Transformation|GetSocketRotation (Variables|Default|GetMesh) "FXSocket") "KeepRelativeOffset" true))
          (Utilities|IsValid _returnvalue
            (:Is%20Valid
              (bind _output_get (Variables|Default|SetFXSystemComponent _returnvalue))
              (Effects|Components|ParticleSystem|SetVectorParameter _output_get "CapVector" (Math|Vector|MakeVector _returnvalue_x _returnvalue_y 2.5e+02)))))))))

(event Custom|SpawnSupplyDrop
  ;[0.02129 0.1089 0.1146] SpawnSupplyDrop: Spawn the supply drop!
  (Networking|SwitchHasAuthority
    (:Authority
      (if (not (Variables|Transient|GetbSupplyDropSpawned))
        (Variables|Transient|SetbSupplyDropSpawned true)
        (bind _data (CallFunction|GetSupplyDropRadioData))
        (if (Utilities|IsValidSoftClassReference (.SupplyDropRadioActorClass_58_71A3A84043A4401BC875CFA344AAF3DD (Utilities|Struct|BreakSupplyDropRadioDataStruct _data)))
          (bind _class (Utilities|AsyncLoadClassAsset (.SupplyDropRadioActorClass_58_71A3A84043A4401BC875CFA344AAF3DD (Utilities|Struct|BreakSupplyDropRadioDataStruct _data)))
            (:Completed
              (bind _asbuilding_gameplay_actor (Utilities|Casting|CastToBuildingGameplayActorClass _class))
              (Building|SpawnBuildingGameplayActorfromWorldContext _asbuilding_gameplay_actor (Math|Conversions|ToTransform(Vector) (Transformation|GetSocketLocation (Variables|Default|GetMesh) "FXSocket")) self (Utilities|Casting|CastToFortPlayerPawn (Actor|GetOwner)))
              (Actor|SetLifeSpan (Math|Float|Max(Float) (ScalableFloat|GetValueatLevel (Variables|Default|GetLifespanAfterSupplyDropSpawn)) 0.1)))))))
    (:Remote
      (Audio|PlaySoundatLocation "/SupplyDropRadioGameplay/Audio/SupplyDropRadio_Flare_Land_Cue.SupplyDropRadio_Flare_Land_Cue" (Transformation|GetActorLocation)))))

(event EventEndPlay
  (params (EndPlayReason :type EEndPlayReason%20Enum :default "Destroyed"))
  ;[0.1927 0.1075 0.01057] EndPlay: Deactivate the flare FX component.
  (|CallParentFunction EndPlayReason)
  (bind _fxsystemcomponent (Variables|Default|GetFXSystemComponent)
    (:then
      (Components|Activation|Deactivate _fxsystemcomponent))
    (:else)))

(fn GetSupplyDropRadioData
  (bind _returnvalue (Fortnite|Athena|GameState|GetGameStateAthena))
  ;[0.07864 0.04767 0.1719] Return the supply drop radio data.
  (Utilities|FlowControl|Sequence
    (:then_0
      (Utilities|IsValid _returnvalue
        (:Is%20Valid
          (bind _output_get (LocalVariables|SetSupplyDropRadioGameStateComponent (Actor|GetComponentbyClass _returnvalue "/SupplyDropRadioGameplay/Gameplay/BP_SupplyDropRadioGameStateComponent.BP_SupplyDropRadioGameStateComponent_C")))
          (Utilities|IsValid _output_get
            (:Is%20Valid
              (bind (_isvalid _data) (Class|BPSupplyDropRadioGameStateComponent|GetSupplyDropRadioDataUsingConversationIndex _output_get (Variables|Transient|GetConversationIndex)))
              (if _isvalid
                (return _data)))))))
    (:then_1
      (return 0))))

(fn HandleProjectileSimulationTimerEnded
  ;[0.009652 0.1426 0.1719] HandleProjectileSimulationTimerEnded: Start the timer to spawn the supply drop.
  (Networking|SwitchHasAuthority
    (:Authority
      (CallFunction|StartSpawnSupplyDropTimer))))

(fn OnRep_bPopped
  ;[0.01145 0.1027 0.1302] Activate the flare on the clients.
  (Networking|CanExecuteCosmeticEvents
    (:True
      (if (Variables|Transient|GetbPopped)
        (CallFunction|ExecutePopCosmetic)))))

(fn OnRep_bThrown
  (bind _mesh (Variables|Default|GetMesh))
  ;[0.01295 0.08346 0.09375] OnRep_bThrown: Water Interaction enabled, setup collision settings for the projectile movement.
  (Variables|/Script/FortniteGame.FortWaterInteractionComponent|SetbIsEnabled true (Variables|Default|GetFortWaterInteraction))
  (Collision|SetCollisionEnabled _mesh "QueryOnly")
  (Variables|/Script/Engine.PrimitiveComponent|SetbGenerateOverlapEvents true _mesh))

(fn PopFlare
  ;[0.1006 0.1615 0.03018] Pop the flare FX.
  ; Called on both server and owning client.
  ; Called from the radio weapon BP.
  (Variables|Transient|SetbPopped true)
  (Networking|CanExecuteCosmeticEvents
    (:True
      (CallFunction|ExecutePopCosmetic))))

(fn SetSupplyDropRadioConversationIndex
  (params (Index :type Integer))
  ;[0.09896 0.05114 0.01868] Cache the conversation index value so it can be used to gather the supply drop class to spawn.
  ; Called from the supply drop game state component.
  (Networking|SwitchHasAuthority
    (:Authority
      (Variables|Transient|SetConversationIndex Index))))

(fn StartSpawnSupplyDropTimer
  ;[0.0077 0.07586 0.1562] Start the timer at the end of which the supply drop actor is spawned.
  (Networking|SwitchHasAuthority
    (:Authority
      (if (not (Utilities|Time|IsTimerActivebyHandle (Variables|Transient|GetSpawnSupplyDropTimerHandle)))
        (bind _returnvalue (Utilities|Time|SetTimerbyEvent (EventDispatchers|CreateEvent :!event SpawnSupplyDrop) (Math|Float|Max(Float) (ScalableFloat|GetValueatLevel (Variables|Default|GetDurationAfterPrjStopToSpawnSupplyDrop)) 0.1)))
        (Variables|Transient|SetSpawnSupplyDropTimerHandle _returnvalue)))))

(fn ThrowFlare
  (params (Velocity :type Vector))
  (bind _fortprojectilemovement (Variables|Default|GetFortProjectileMovement))
  ;[0.07042 0.1771 0.03874] ThrowFlare: Activate the projectile movement component and set the desired velocity.
  ; Called from the radio weapon BP.
  (Networking|SwitchHasAuthority
    (:Authority
      (if (not (Variables|Transient|GetbThrown))
        (Variables|Transient|SetbThrown true)
        (Components|Activation|Activate _fortprojectilemovement false)
        (Projectile|SetVelocity _fortprojectilemovement Velocity)
        (Networking|ForceNetUpdate)
        (Variables|Transient|SetThrowRotation (Math|Rotator|MakeRotator 9e+01 (Math|Random|RandomFloatinRange 0.0 1.8e+02) (Math|Random|RandomFloatinRange 0.0 1.8e+02))) ; The rotation to interp to.
        ;;[0.0151 0.08911 0.1302] Start a timer after which the simulation is auto-ended and the supply drop is spawned.
        (bind _returnvalue (Utilities|Time|SetTimerbyEvent (EventDispatchers|CreateEvent :!event HandleProjectileSimulationTimerEnded) (Math|Float|Max(Float) (ScalableFloat|GetValueatLevel (Variables|Default|GetMaxProjectileSimulationTime)) 0.1)))
        (Variables|Transient|SetProjectileSimulationTimerHandle _returnvalue)))))

(fn ConstructionScript)
