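; Handheld supply drop radio weapon item that manages a flare actor held in the owning player's left hand.
; The server spawns and attaches the flare on equip or RespawnFlare, and destroys it on unequip or EndPlay if unthrown.
; The activate gameplay ability calls PopFlare and ThrowFlare, which launches it server-side toward the ground ahead using hotfixable speed and distance.
; Flare visibility follows the weapon's, and holstering is disabled.
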(event FortWeapon|EventOnEquippedAndAttached
  ;[0 0.2396 0.000868] OnEquippedAndAttached: Spawn and attach the flare to the owner player's left hand.
  (|CallParentFunction)
  (Networking|SwitchHasAuthority
    (:Authority
      (Utilities|FlowControl|Sequence
        (:then_0
          (bind _flareactor (Variables|Transient|GetFlareActor)
            (:then
              (Actor|DestroyActor _flareactor)
              (Fortnite|FortShippingLog self "OnEquippedAndAttached: Flare actor was valid when it shouldn't be." true))
            (:else)))
        (:then_1
          (bind _asfort_player_pawn (Utilities|Casting|CastToFortPlayerPawn (Actor|GetOwner)))
          (bind _mesh (Variables|/Script/Engine.Character|GetMesh _asfort_player_pawn)
            (:then
              (bind _returnvalue (Game|SpawnActorfromClass (Variables|Settings|GetSupplyDropRadioFlareActorClass) (Math|Transform|MakeTransform (Transformation|GetActorLocation)) "AlwaysSpawn" "OverrideRootScale" _asfort_player_pawn _asfort_player_pawn))
              (bind _output_get (Variables|Transient|SetFlareActor _returnvalue))
              (Transformation|AttachActorToComponent _output_get _mesh "hand_attach_l" "SnapToTarget" "SnapToTarget" "KeepWorld" true))
            (:else)))))))

(event Custom|RespawnFlare
  ;[0 0.2396 0.000868] OnEquippedAndAttached: Spawn and attach the flare to the owner player's left hand.
  (Networking|SwitchHasAuthority
    (:Authority
      (Variables|Transient|SetbFlareThrown false)
      (Utilities|FlowControl|Sequence
        (:then_0
          (bind _flareactor (Variables|Transient|GetFlareActor)
            (:then
              (Actor|DestroyActor _flareactor)
              (Fortnite|FortShippingLog self "OnEquippedAndAttached: Flare actor was valid when it shouldn't be." true))
            (:else)))
        (:then_1
          (bind _asfort_player_pawn (Utilities|Casting|CastToFortPlayerPawn (Actor|GetOwner)))
          (bind _mesh (Variables|/Script/Engine.Character|GetMesh _asfort_player_pawn)
            (:then
              (bind _returnvalue (Game|SpawnActorfromClass (Variables|Settings|GetSupplyDropRadioFlareActorClass) (Math|Transform|MakeTransform (Transformation|GetActorLocation)) "AlwaysSpawn" "OverrideRootScale" _asfort_player_pawn _asfort_player_pawn))
              (bind _output_get (Variables|Transient|SetFlareActor _returnvalue))
              (Transformation|AttachActorToComponent _output_get _mesh "hand_attach_l" "SnapToTarget" "SnapToTarget" "KeepWorld" true))
            (:else)))))))

(event FortWeapon|EventOnUnEquippedAndDetached
  ;[0.01204 0.1323 0.1615] OnUnequippedWeapon/EndPlay: IF the flare is not yet thrown, destroy it.
  (|CallParentFunction)
  (Networking|SwitchHasAuthority
    (:Authority
      (if (not (Variables|Transient|GetbFlareThrown))
        (bind _flareactor (Variables|Transient|GetFlareActor)
          (:then
            (Actor|DestroyActor _flareactor))
          (:else))))))

(event EventEndPlay
  (params (EndPlayReason :type EEndPlayReason%20Enum :default "Destroyed"))
  ;[0.01204 0.1323 0.1615] OnUnequippedWeapon/EndPlay: IF the flare is not yet thrown, destroy it.
  (|CallParentFunction EndPlayReason)
  (Networking|SwitchHasAuthority
    (:Authority
      (if (not (Variables|Transient|GetbFlareThrown))
        (bind _flareactor (Variables|Transient|GetFlareActor)
          (:then
            (Actor|DestroyActor _flareactor))
          (:else))))))

(event Custom|ThrowFlare
  (bind (_returnvalue_x _returnvalue_y _returnvalue_z) (Math|Vector|BreakVector (Transformation|GetActorLocation (Utilities|Casting|CastToFortPlayerPawn (Actor|GetOwner)))))
  (bind (_outradius _outhalfheight) (Components|Capsule|GetScaledCapsuleSize (Variables|/Script/Engine.Character|GetCapsuleComponent (Utilities|Casting|CastToFortPlayerPawn (Actor|GetOwner)))))
  (bind _returnvalue (Actor|GetOwner))
  (bind _asfort_player_pawn (Utilities|Casting|CastToFortPlayerPawn _returnvalue))
  ;[0.1091 0.01088 0.1094] ThrowFlare: Throw the flare on to the ground next to the player.
  ; This event is called from the supply drop radio activate gameplay ability.
  ; Server only.
  (Networking|SwitchHasAuthority
    (:Authority
      (if (not (Variables|Transient|GetbFlareThrown))
        (Variables|Transient|SetbFlareThrown true)
        (bind _flareactor (Variables|Transient|GetFlareActor)
          (:then
            (Transformation|DetachFromActor _flareactor "KeepWorld" "KeepWorld" "KeepWorld")
            (Transformation|SetActorLocation _flareactor (Math|Transform|TransformLocation (Transformation|GetActorTransform _asfort_player_pawn) (select (Variables|/Script/Engine.Character|GetbIsCrouched _asfort_player_pawn) (Variables|Settings|GetServerThrowStartRelativeLocationCrouched) (Variables|Settings|GetServerThrowStartRelativeLocation))) false false)
            (Class|BPSupplyDropRadioFlareActor|ThrowFlare _flareactor (* (Math|Vector|GetUnitDirection(Vector) (Transformation|GetActorLocation _flareactor) (+ (* (Math|Vector|Normalize2D(Vector) (Transformation|GetActorForwardVector _asfort_player_pawn)) (ScalableFloat|GetValueatLevel (Variables|Hotfix|GetFlareThrowForwardDistance))) (Math|Vector|MakeVector _returnvalue_x _returnvalue_y (- _returnvalue_z _outhalfheight)))) (ScalableFloat|GetValueatLevel (Variables|Hotfix|GetFlareThrowSpeed))))
            (Variables|Transient|SetFlareActor 0) ; Forget about the flare.
          )
          (:else))))))

(event EventOnWeaponVisibilityChanged
  (params (bVisible :type Boolean) (bSetForLocalControllerOnly :type Boolean))
  ;[0.00856 0.1138 0.1198] OnWeaponVisibilityChanged: Adjust the visibility of the flare actor to match the visibility of this weapon.
  (|CallParentFunction bVisible bSetForLocalControllerOnly)
  (Networking|CanExecuteCosmeticEvents
    (:True
      (if bSetForLocalControllerOnly
        (bind _asfort_player_pawn (Utilities|Casting|CastToFortPlayerPawn (Actor|GetOwner)))
        (if (Pawn|IsLocallyControlled _asfort_player_pawn)
          _)
        (else
          _))
      (bind _flareactor (Variables|Transient|GetFlareActor)
        (:then
          (Rendering|SetActorHiddenInGame _flareactor (not bVisible)))
        (:else)))))

(event Custom|PopFlare
  ;[0.2448 0.04142 0.1435] Pop the flare actor.
  ; Called from the activate gameplay ability.
  (bind _flareactor (Variables|Transient|GetFlareActor)
    (:then
      (Class|BPSupplyDropRadioFlareActor|PopFlare _flareactor))
    (:else)))

(fn CanApplyHolster
  (params (HolsterId :type Name%20%28by%20ref%29 :default "None"))
  (return false))

(fn GetFlareActor
  (return (Variables|Transient|GetFlareActor)))

(fn ConstructionScript
  (|CallParentFunction))
