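; Gameplay ability for the handheld Supply Drop Radio's primary fire, letting players call in supply drops with a thrown flare.
; It plays an activation montage, starts a server conversation, then reserves the selected drop through a GameState component, unless configured to throw directly.
; Confirm montage events pop and throw the flare, copying the reservation onto it; a timer force-stops the montage to restore the bone-scaled radio.
; Ending refunds unused reservations and resets for further uses.
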(event Ability|EventActivateAbilityFromEvent
  (params (EventData :type Gameplay%20Event%20Data%20Structure%20%28by%20ref%29))
  (bind _returnvalue (Ability|GetCurrentSourceWeapon))
  (bind _returnvalue_2 (Actor|GetComponentbyClass _returnvalue "/Script/FortniteGame.FortNonPlayerConversationParticipantComponent"))
  (bind _supplydropradio (Default|GetCurrentSupplyDropRadio))
  (bind _returnvalue_3 (CallFunction|GetAvatarasFortPlayerPawn))
  (bind _montage_confirm (Variables|Default|GetMontage_Confirm))
  ;[0.01197 0.1667 0.0483] ActivateAbilityFromEvent: Called when the radio is activated to start a conversation.
  ; This conversation provides options for the player to select a type of supply drop.
  ; Once selected, the conversation systems sends an event along with a payload holding information about the selected option.
  ; This information is then used to call in for the supply drop while playing the necessary animations.
  (|CallParentFunction EventData)
  (bind _outfortplayerpawn (CallFunction|GetAvatarasValidFortPlayerPawn)
    (:ValidFortPlayerPawn
      (CallFunction|HandlePlayerMovementStates)
      (FortPawn|BindEventtoOnWeaponUnEquippedDelegate _outfortplayerpawn (EventDispatchers|CreateEvent :!event Custom|HandleWeaponUnequipped))
      (if (Variables|Default|GetbSkipConvoSelectDirectThrow)
        ;;[0.6031 0.3242 0.4335 0.5] The montage scales the 'hand_attach_r' bone down to 0.001 to hide the radio, and it stays that way until the montage is forced to stop. If the radio has infinite uses (possible in Creative/UEFN), it'll still be like that when you try to use it again.
        ;;
        ;; The Completed pin won't execute after the montage fires its trigger, so we need a timer to know when the animation is done to stop the montage. That will trigger the Cancelled pin and end the ability, to then call ResetToHandleAdditionalUses().
        (bind _asynctaskproxy (Ability|Tasks|PlayMontageandWaitforTrigger "Task_MontageConfirm" _montage_confirm 1.0 "FirstSection" "None" 1.0 "SimulateOnServer" :!label _label)
          (:then
            (Variables|Transient|SetAsyncTask_MontageConfirm _asynctaskproxy)
            (bind _returnvalue_4 (Utilities|Time|SetTimerbyEvent (delegate Custom|EndConfirmMontage) (Animation|GetPlayLength _montage_confirm)))
            (Variables|Default|SetMontageTimer_Confirm _returnvalue_4)
            (Utilities|FlowControl|Sequence
              (:then_0
                (bind _asynctaskproxy_2 (Ability|Async|WaitGameplayEventToActor _returnvalue_3 "(TagName=\"SupplyDropRadio.Anim.Event.Flare.Throw\")" true)
                  (:then
                    (Variables|Transient|SetAsyncTask_AnimEventFlareThrow _asynctaskproxy_2))
                  (:EventReceived
                    (Ability|ExecuteGameplayCueOnOwner (Variables|Default|GetForceFeedbackCueTag))
                    (bind _returnvalue_5 (Ability|CommitAbility))
                    ;;[1 1 1] Ensure item removed from inventory successfully
                    (if _returnvalue_5
                      (Utilities|IsServer
                        (:true
                          (if (not (Variables|Transient|GetbSupplyDropClassSetOnFlare))
                            (if (not (Variables|Default|GetbSkipConvoSelectDirectThrow))
                              (bind _component (Utilities|GetSupplyDropRadioGameStateComponent)
                                (:Valid
                                  ;;[0.08681 0.0434 0.1302] Copy the reserved data to the flare actor so that it knows which supply drop class to spawn.
                                  (bind _bsuccess (Class|BPSupplyDropRadioGameStateComponent|SetReservedSupplyDroptoFlare _component (CallFunction|GetAvatarasFortPlayerPawn) (CallFunction|GetCurrentFlareActor)))
                                  (Variables|Transient|SetbSupplyDropClassSetOnFlare _bsuccess))))
                            (Utilities|IsValid _supplydropradio
                              (:Is%20Valid
                                (Class|BSupplyDropRadioHandheldItem|ThrowFlare _supplydropradio))
                              (:Is%20Not%20Valid
                                (Fortnite|FortShippingLog self "ActivateAbilityFromEvent: SupplyDropRadio Handheld Item not equipped on the player." true)))))
                        (:false
                          (CallFunction|DebugThrowSocketLocation)))
                      (else
                        _)))))
              (:then_1
                ;;[0.1406 0.02104 0.1387] Pop the flame on the flare.
                (bind _asynctaskproxy_3 (Ability|Async|WaitGameplayEventToActor _returnvalue_3 "(TagName=\"SupplyDropRadio.Anim.Event.Flare.Pop\")" true)
                  (:then
                    (Variables|Transient|SetAsyncTask_AnimEventFlarePop _asynctaskproxy_3))
                  (:EventReceived
                    (bind _asb_supply_drop_radio_handheld_item (Utilities|Casting|CastToB_SupplyDropRadio_Handheld_Item (Weapon|GetCurrentWeapon (CallFunction|GetAvatarasFortPlayerPawn))))
                    (Class|BSupplyDropRadioHandheldItem|PopFlare _asb_supply_drop_radio_handheld_item))))))
          (:Cancelled
            _)
          (:Completed
            _))
        (else
          (Utilities|IsValid _returnvalue
            (:Is%20Valid
              (Utilities|IsValid _returnvalue_2
                (:Is%20Valid
                  (EventDispatchers|BindEventtoOnConversationEnded _returnvalue_2 (EventDispatchers|CreateEvent :!event Custom|HandleConversationEnded))
                  (if (Variables|Settings|GetbStopMovementForBriefDurationOnActivate)
                    (bind _returnvalue_6 (Ability|ApplyGameplayEffectToOwner "/SupplyDropRadioGameplay/Items/SupplyDropRadio_Handheld/GE_SupplyDropRadio_BlockMovement.GE_SupplyDropRadio_BlockMovement_C")) ; Stop movement for a bit.
                    (Variables|Transient|SetBlockMovementEffectHandle _returnvalue_6))
                  (bind _asynctaskproxy_4 (Ability|Tasks|PlayMontageandWaitforTrigger "Task_MontageActivate" "/SupplyDropRadioGameplay/Items/SupplyDropRadio_Handheld/Animation/SupplyDropRadioActivateMontage.SupplyDropRadioActivateMontage" 1.0 "FirstSection" "None" 1.0 "SimulateOnServer")
                    (:then
                      (Variables|Transient|SetAsyncTask_MontageActivate _asynctaskproxy_4))
                    (:Triggered
                      (Utilities|IsServer
                        (:true
                          (Conversation|StartConversation _returnvalue_2 "(TagName=\"Conversation.Entry.Item.SupplyDropRadio\")" (Pawn|GetController _outfortplayerpawn) _returnvalue :!class /Script/FortniteGame.FortNonPlayerConversationParticipantComponent)
                          (if (Conversation|IsinActiveConversation _returnvalue_2)
                            (bind (_asynctaskproxy_5 _payload) (Ability|Tasks|WaitGameplayEvent "(TagName=\"Event.SupplyDropRadio.Conversation.Option.Select\")" :!label _label_4)
                              (:then
                                (Variables|Transient|SetAsyncTask_OptionSelect _asynctaskproxy_5))
                              (:EventReceived
                                (if (Variables|Default|GetbReactToConvoSelect)
                                  (Variables|Default|SetbReactToConvoSelect false)
                                  (Variables|Transient|SetSelectedConversationIndex (Math|Float|Truncate (.EventMagnitude (Utilities|Struct|BreakGameplayEventData _payload))))
                                  (Utilities|IsServer
                                    (:true
                                      (bind _component_2 (Utilities|GetSupplyDropRadioGameStateComponent)
                                        (:Valid
                                          (bind _bsuccessful (Class|BPSupplyDropRadioGameStateComponent|ReserveSupplyDrop _component_2 (CallFunction|GetAvatarasFortPlayerPawn) (Variables|Transient|GetSelectedConversationIndex)))
                                          (if _bsuccessful
                                            (CallFunction|UnbindConversationEndDelegate :!label _label_2)
                                            (goto _label)
                                            (else
                                              _)))
                                        (:Invalid
                                          _))
                                      ;;[1 0.5 0.5] End ability
                                      (Ability|EndAbility :!label _label_3))
                                    (:false
                                      _))
                                  (goto _label_2)))))
                          (goto _label_3))
                        (:false
                          (Utilities|FlowControl|Sequence
                            (:then_0
                              (bind _asynctaskproxy_6 (Ability|Tasks|WaitGameplayTagAdd "(TagName=\"Conversation.InteractingWithNPC\")")
                                (:then
                                  (Variables|Transient|SetAsyncTask_InteractingWithNPC _asynctaskproxy_6))
                                (:Added
                                  (Conversation|SetMoveShouldAbortConversation (Utilities|Casting|CastToFortPlayerControllerGameplay (Pawn|GetController _outfortplayerpawn)) (Variables|Settings|GetbMoveAbortConversation)))))
                            (:then_1
                              _))))
                      (goto _label_4))
                    (:Cancelled
                      (if (== (Variables|Transient|GetSelectedConversationIndex) -1) :!label _label_5
                        _))
                    (:Completed
                      (goto _label_5))))
                (:Is%20Not%20Valid
                  _)))
            (:Is%20Not%20Valid
              _)))))
    (:AvatarCastFailed
      _))
  (goto _label_3))

(event Custom|EndConfirmMontage
  ;[0.01197 0.1667 0.0483] ActivateAbilityFromEvent: Called when the radio is activated to start a conversation.
  ; This conversation provides options for the player to select a type of supply drop.
  ; Once selected, the conversation systems sends an event along with a payload holding information about the selected option.
  ; This information is then used to call in for the supply drop while playing the necessary animations.
  ;;[0.6031 0.3242 0.4335 0.5] The montage scales the 'hand_attach_r' bone down to 0.001 to hide the radio, and it stays that way until the montage is forced to stop. If the radio has infinite uses (possible in Creative/UEFN), it'll still be like that when you try to use it again.
  ;;
  ;; The Completed pin won't execute after the montage fires its trigger, so we need a timer to know when the animation is done to stop the montage. That will trigger the Cancelled pin and end the ability, to then call ResetToHandleAdditionalUses().
  (bind _outfortplayerpawn (CallFunction|GetAvatarasValidFortPlayerPawn)
    (:ValidFortPlayerPawn
      (Animation|Utilities|StopMontageifCurrent _outfortplayerpawn (Variables|Default|GetMontage_Confirm)) ; Blend out the montage with its default blend time.
    )))

(event Custom|HandleConversationEnded
  ;[0 0.1875 0.1522] HandleConversationEnded: End the ability when the conversation ends.
  (Variables|Transient|SetbConversationEnded true)
  (Ability|EndAbility))

(event Custom|HandleWeaponUnequipped
  ;[0 0.2135 0.2009] HandleWeaponUnequipped: Called when the weapon is unequipped.
  (Ability|EndAbility))

(event Ability|EventOnEndAbility
  (params (bWasCancelled :type Boolean))
  ;[0.09896 0.0537 0] EndAbility: Unbind conversation end delegate.
  (|CallParentFunction bWasCancelled)
  (GameplayTasks|EndTask (Variables|Transient|GetAsyncTask_InteractingWithNPC))
  (Utilities|Time|ClearandInvalidateTimerbyHandle (Variables|Default|GetMontageTimer_Confirm))
  (CallFunction|ResettoHandleAdditionalUses)
  (CallFunction|UnbindConversationEndDelegate)
  (Ability|RemoveGameplayEffectFromOwnerWithHandle (Variables|Transient|GetBlockMovementEffectHandle))
  (bind _outfortplayerpawn (CallFunction|GetAvatarasValidFortPlayerPawn)
    (:ValidFortPlayerPawn
      (FortPawn|UnbindEventfromOnWeaponUnEquippedDelegate _outfortplayerpawn (EventDispatchers|CreateEvent :!event Custom|HandleWeaponUnequipped))
      (Character|UnbindEventfromMovementModeChangedDelegate _outfortplayerpawn (EventDispatchers|CreateEvent :!event Custom|HandleMovementModeChanged))
      (Conversation|SetMoveShouldAbortConversation (Utilities|Casting|CastToFortPlayerControllerGameplay (Pawn|GetController _outfortplayerpawn)) true)
      (Utilities|FlowControl|Sequence
        (:then_0
          (Utilities|IsServer
            (:true
              (bind _asynctask_animeventflarethrow (Variables|Transient|GetAsyncTask_AnimEventFlareThrow)
                (:then
                  (Ability|Async|EndAction _asynctask_animeventflarethrow))
                (:else
                  _))
              (if (Variables|Transient|GetbConversationEnded)
                (Variables|Transient|SetbConversationEnded false)
                (else
                  (Conversation|StopCurrentConversation (Utilities|Casting|CastToFortPlayerControllerGameplay (Pawn|GetController _outfortplayerpawn)))))
              ;;[0.1302 0.1007 0.05642] IF the player selected a supply drop option but a supply drop class wasn't successfully set, refund the supply drop stock.
              (if (and (not (Variables|Transient|GetbSupplyDropClassSetOnFlare)) (>= (Variables|Transient|GetSelectedConversationIndex) 0))
                (bind _component (Utilities|GetSupplyDropRadioGameStateComponent)
                  (:Valid
                    (Class|BPSupplyDropRadioGameStateComponent|RefundSupplyDrop _component (CallFunction|GetAvatarasFortPlayerPawn))))))))
        (:then_1
          (bind _asynctask_animeventflarepop (Variables|Transient|GetAsyncTask_AnimEventFlarePop)
            (:then
              (Ability|Async|EndAction _asynctask_animeventflarepop))
            (:else)))))))

(event Custom|HandleMovementModeChanged
  (params (Character :type Character%20Object%20Reference) (PrevMovementMode :type EMovementMode%20Enum :default "MOVE_None") (PreviousCustomMode :type Byte))
  ;[0.07162 0.3125 0.3125] HandleMovementModeChanged: End ability if the movement mode ever changes to one that is not allowed.
  (Utilities|IsValid Character
    (:Is%20Valid
      (bind _asfort_player_pawn (Utilities|Casting|CastToFortPlayerPawn Character))
      (bind _charactermovement (Variables|/Script/Engine.Character|GetCharacterMovement _asfort_player_pawn)
        (:then
          (if (or (not (CallFunction|IsMovementComponenetinAllowedMode _charactermovement)) (Parachute|IsSkydiving _asfort_player_pawn))
            (Ability|EndAbility)))
        (:else)))))

(event Custom|ResetToHandleAdditionalUses
  (bind _supplydropradio (Default|GetCurrentSupplyDropRadio))
  (bind _returnvalue (FortWeapon|GetInventoryItem _supplydropradio))
  ;[0.6198 0.2193 0.1707 0.5] ResetToHandleAdditionalUses: Reset the bone changes from the montage, spawn a new flare in the player's hand, and reset state-tracking variables.
  (Utilities|IsValid _supplydropradio
    (:Is%20Valid
      (Utilities|IsValid _returnvalue
        (:Is%20Valid
          (bind _outfortplayerpawn (CallFunction|GetAvatarasValidFortPlayerPawn)
            (:ValidFortPlayerPawn
              (if (> (Item|GetNuminStackwithInfinityCheck _returnvalue) 0)
                (Animation|Utilities|StopMontageifCurrent _outfortplayerpawn (Variables|Default|GetMontage_Confirm) 0.0) ; Override the Blend Out Time to immediately stop\nthe montage to reset its changes to the player.\n\nClean up the montage on the client, and handle any\nweirdness on the server if the ability ended early.
                (Class|BSupplyDropRadioHandheldItem|RespawnFlare _supplydropradio)
                (Variables|Transient|SetbSupplyDropClassSetOnFlare false)
                (Variables|Default|SetbReactToConvoSelect true)))))))))

(fn DebugThrowSocketLocation
  ;[0.04177 0.2218 0.2812] DebugThrowSocketLocation: Used to debug the location of the left hand throw socket so the value can be set in the supply drop radio weapon's throw logic.
  (Networking|CanExecuteCosmeticEvents
    (:True
      (if (and (Editor|InEditororPIE) (Variables|Settings|GetbDebugThrowSocketLocation))
        (bind _outfortplayerpawn (CallFunction|GetAvatarasValidFortPlayerPawn)
          (:ValidFortPlayerPawn
            (bind _mesh (Variables|/Script/Engine.Character|GetMesh _outfortplayerpawn)
              (:then
                (Rendering|Debug|DrawDebugSphere (Transformation|GetSocketLocation _mesh "hand_attach_l") 1e+01 12 "(R=0.944346,G=0.105561,B=1.000000,A=1.000000)" 1e+01 0.2)
                (Development|PrintString (Utilities|String|Append "GA_SupplyDropRadio: DebugThrowSocketLocation: " (Utilities|String|ToString(Vector) (Math|Transform|InverseTransformLocation (Transformation|GetActorTransform _outfortplayerpawn) (Transformation|GetSocketLocation _mesh "hand_attach_l")))) false))
              (:else))))))))

(fn GetCurrentFlareActor
  (bind _supplydropradio (Default|GetCurrentSupplyDropRadio))
  (Utilities|IsValid _supplydropradio
    (:Is%20Valid
      (return (Class|BSupplyDropRadioHandheldItem|GetFlareActor _supplydropradio :!class /SupplyDropRadioGameplay/Items/SupplyDropRadio_Handheld/B_SupplyDropRadio_Handheld_Item.B_SupplyDropRadio_Handheld_Item_C)))
    (:Is%20Not%20Valid
      (return 0))))

(fn GetCurrentSupplyDropRadio
  (bind _returnvalue (CallFunction|GetAvatarasFortPlayerPawn))
  (Utilities|IsValid _returnvalue
    (:Is%20Valid
      (bind _asb_supply_drop_radio_handheld_item (Utilities|Casting|CastToB_SupplyDropRadio_Handheld_Item (Weapon|GetCurrentWeapon _returnvalue))
        (:then
          (return _asb_supply_drop_radio_handheld_item))
        (:CastFailed
          _)))
    (:Is%20Not%20Valid
      _))
  (return 0))

(fn HandlePlayerMovementStates
  (bind _outfortplayerpawn (CallFunction|GetAvatarasValidFortPlayerPawn)
    (:ValidFortPlayerPawn
      (Character|BindEventtoMovementModeChangedDelegate _outfortplayerpawn (EventDispatchers|CreateEvent :!event HandleMovementModeChanged)))))

(fn IsMovementComponenetInAllowedMode
  (params (CharacterMovementComponent :type Character%20Movement%20Component%20Object%20Reference))
  (Utilities|IsValid CharacterMovementComponent
    (:Is%20Valid
      (return (Math|Boolean|ORBoolean (Utilities|Enum|Equal(Enum) (Utilities|Enum|BytetoEnumEFortCustomMovement (Variables|/Script/Engine.CharacterMovementComponent|GetCustomMovementMode CharacterMovementComponent)) "Sliding") (Utilities|Enum|Equal(Enum) (Utilities|Enum|BytetoEnumEFortCustomMovement (Variables|/Script/Engine.CharacterMovementComponent|GetCustomMovementMode CharacterMovementComponent)) "Ziplining") (Utilities|Enum|Equal(Enum) (Utilities|Enum|BytetoEnumEFortCustomMovement (Variables|/Script/Engine.CharacterMovementComponent|GetCustomMovementMode CharacterMovementComponent)) "ZipliningOnSpline") (Pawn|Components|CharacterMovement|IsWalking CharacterMovementComponent) (AI|Components|NavMovement|IsFalling CharacterMovementComponent))))
    (:Is%20Not%20Valid
      (return false))))

(fn CanActivateAbility
  (params (ActorInfo :type Gameplay%20Ability%20Actor%20Info%20Structure) (Handle :type Gameplay%20Ability%20Spec%20Handle%20Structure))
  (bind _returnvalue (CallFunction|GetAvatarasFortPlayerPawn))
  ;[0.03546 0.2135 0.06946] Ability can be activated as long as the player is standing on the ground.
  (Utilities|FlowControl|Sequence
    (:then_0
      (Utilities|IsValid _returnvalue
        (:Is%20Valid
          (if (not (Health|IsDBNO _returnvalue))
            (if (not (Vehicle|IsinVehicle _returnvalue))
              (if (not (GameplayTags|HasAnyMatchingGameplayTags _returnvalue (GameplayTags|MakeLiteralGameplayTagContainer "(GameplayTags=((TagName=\"Gameplay.Action.Player.Ascender.Ascending\"),(TagName=\"Gameplay.Action.Player.Ascender.Descending\")))")))
                (if (not (Parachute|IsSkydiving _returnvalue))
                  (bind _asfort_movement_comp_character_athena (Utilities|Casting|CastToFortMovementComp_CharacterAthena (Variables|/Script/Engine.Character|GetCharacterMovement _returnvalue)))
                  (return 0 (or (CallFunction|IsMovementComponenetinAllowedMode _asfort_movement_comp_character_athena) (GameplayTags|HasMatchingGameplayTag _returnvalue "(TagName=\"Riding.IsRidingActor\")"))))))))))
    (:then_1
      (return 0 false))))

(fn OverrideFailedReason
  (params (FailedReason :type Gameplay%20Tag%20Container%20Structure%20%28by%20ref%29))
  (return "(GameplayTags=((TagName=\"Athena.Ability.Failed.NoMessage\")))"))

(fn ShouldPlayFailedMontage
  (params (FailedReason :type Gameplay%20Tag%20Container%20Structure%20%28by%20ref%29))
  (return true))

(fn UnbindConversationEndDelegate
  (bind _returnvalue (Ability|GetCurrentSourceWeapon))
  (bind _returnvalue_2 (Actor|GetComponentbyClass _returnvalue "/Script/FortniteGame.FortNonPlayerConversationParticipantComponent"))
  (Utilities|IsValid _returnvalue
    (:Is%20Valid
      (Utilities|IsValid _returnvalue_2
        (:Is%20Valid
          (EventDispatchers|UnbindEventfromOnConversationEnded _returnvalue_2 (EventDispatchers|CreateEvent :!event HandleConversationEnded)))))))
